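; Egg grenade launcher projectile Blueprint that handles bounce, water-entry and explosion feedback, and shows sound indicators for all affiliations.
; Bounces spawn FX and collision audio, water entries fire up to three splash cues, and stopping deactivates rotating movement.
; Exploding picks a water or dry GameplayCue, the latter reusing the last bounce normal, and stops the looping sound indicator.
; An AllowJanus-gated, CVar-checked branch also spawns a radial-force physics field.
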(event EventBeginPlay
  (|CallParentFunction)
  (Variables|Default|SetSplashCounter 0))

(event Collision|EventActorBeginOverlap
  (params (OtherActor :type Actor%20Object%20Reference)))

(event EventTick
  (params (DeltaSeconds :type Float%20%28single-precision%29)))

(event EventOnStop
  (params (Hit :type Hit%20Result%20Structure%20%28by%20ref%29))
  (|CallParentFunction Hit)
  (Components|Activation|Deactivate (Variables|Default|GetRotatingMovement)))

(event EventOnExploded
  (params (HitActors :type Array%20of%20Actor%20Object%20References) (HitResults :type Array%20of%20Hit%20Result%20Structures))
  (bind _capsulecomponent (Variables|Collision|GetCapsuleComponent))
  (bind _returnvalue (Transformation|GetWorldTransform _capsulecomponent))
  (bind _location (.location _returnvalue))
  (|CallParentFunction HitActors HitResults)
  (Utilities|FlowControl|Sequence
    (:then_0
      (Utilities|AllowJanus
        (:Allow
          (bind _returnvalue_2 (Development|GetConsoleVariableIntValue "r.PhysicsField.EnableField"))
          (if (== _returnvalue_2 1)
            (Game|SpawnActorfromClass "/Game/Effects/Niagara/Blueprints/PhsyicsFields/FN_RadialForce.FN_RadialForce_C" _returnvalue "Undefined" "MultiplyWithRoot" 0 1.0 0.75 5e+02 1.0 0.0 0)))
        (:Nope
          _))
      (if (Water|IsinWater (Variables|Water|GetWaterInteractionComponent))
        (Development|PrintString "InWater")
        (GameplayCue|ExecuteGameplayCue%28Burst%29Local self (Variables|Default|GetExplosion_Water_GC) (Ability|GameplayCue|MakeGameplayCueParameters :Location _location :Normal "0,0,1.000000") true)
        (else
          (Development|PrintString "Out Of Water")
          (GameplayCue|ExecuteGameplayCue%28Burst%29Local self (Variables|Default|GetExplosion_GC) (Ability|GameplayCue|MakeGameplayCueParameters :Location _location :Normal (.ImpactNormal (Collision|BreakHitResult (stale EventOnBounce Hit)))) true))))
    (:then_1
      (SoundIndicator|StopSoundIndicatorLooping (Variables|Default|GetFortSoundIndicator)))))

(event OnEnterWaterEvent(WaterInteractionComponent)
  (params (WaterBody :type Fort%20Water%20Body%20Actor%20Object%20Reference) (WaterInteractionComponent :type Fort%20Water%20Interaction%20Component%20Object%20Reference) (bIsFirstBody :type Boolean))
  (if (not (>= (Variables|Default|GetSplashCounter) 3))
    (GameplayCue|ExecuteGameplayCue%28Burst%29Local self (Variables|Default|GetImpactWater_GC) (struct :NormalizedMagnitude 0.0 :RawMagnitude 0.0 :Location "0,0,0" :Normal "0,0,1.000000" :Instigator 0 :EffectCauser 0 :SourceObject 0 :PhysicalMaterial 0 :GameplayEffectLevel 1 :AbilityLevel 1 :TargetAttachComponent 0 :bReplicateLocationWhenUsingMinimalRepProxy false) true)
    (Variables|Default|SetSplashCounter (+ (Variables|Default|GetSplashCounter) 1))))

(event EventOnBounce
  (params (Hit :type Hit%20Result%20Structure%20%28by%20ref%29))
  (bind (_bblockinghit _binitialoverlap _time _distance _location _impactpoint _normal _impactnormal) (Collision|BreakHitResult Hit))
  (bind _hitactor (CallFunction|SpawnBounceFX Hit))
  (Variables|Default|SetSpawnBounceFXHitActor _hitactor)
  (Utilities|Casting|CastToPawn (Variables|Default|GetSpawnBounceFXHitActor))
  (Audio|OnCollision (Variables|Default|GetFortCollisionAudio) _impactpoint (* _impactnormal (Variables|/Script/Engine.MovementComponent|GetVelocity (Variables|Movement|GetProjectileMovementComponent)))))

(fn GetAffiliationsToShowFor
  (return (Utilities|Array|MakeArray "Friendly" "Neutral" "Hostile")))

(fn ShouldShowSoundIndicator
  (params (PlayerController :type Fort%20Player%20Controller%20Object%20Reference))
  (return true))

(fn ConstructionScript
  (|CallParentFunction))
