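; Snowball grenade launcher projectile Blueprint that trails a Niagara ribbon and explodes on hit, on entering water or when its fuse timer expires.
; Stopping caches location, rotation and surface type, swapping the mesh for a landing-delay effect on water.
; The cosmetic-only explosion handler spawns Niagara, feeding it water surface data when submerged, then plays sound and a camera-shake cue.
; SpawnBounceFX and the delayed-explode event have no callers here.
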(event EventBeginPlay
  ;[1 1 1] SpawnRibbonTrails
  (bind _returnvalue (Niagara|SpawnSystemAttached (Variables|Default|GetNS_Ribbon) (Variables|Projectile_GrenadeLauncher|GetMesh)))
  (Components|Activation|Activate _returnvalue false)
  (Effects|Components|ParticleSystem|SetFloatParameter _returnvalue "TrailLifetime" (Variables|Effects|GetTrailLifetime))
  (Transformation|SetRelativeScale3D (Variables|Projectile_GrenadeLauncher|GetMesh) "2.000000,2.000000,2.000000")
  (Utilities|Time|SetTimerbyFunctionName self "FuseTimerMax" (Variables|Default|GetFuseTimer@Spawn)))

(event EventOnStop
  (params (Hit :type Hit%20Result%20Structure%20%28by%20ref%29))
  (bind (_bblockinghit _binitialoverlap _time _distance _location _impactpoint _normal _impactnormal _physmat) (Collision|BreakHitResult Hit))
  ;[1 1 1] StopBehavior, get PhysMat
  (Components|Activation|Deactivate (Variables|Default|GetRotatingMovement))
  (Variables|Default|SetStopLocation _impactpoint) ; Use Loc and Rot for single explosion
  (Variables|Default|SetStopRotX (Math|Rotator|MakeRotfromX _impactnormal))
  (Variables|Default|SetStopRotZ (Math|Rotator|MakeRotfromZ _impactnormal))
  (Variables|Effects|SetStopSurfaceType (Variables|/Script/PhysicsCore.PhysicalMaterial|GetSurfaceType _physmat))
  (Variables|Default|SetHasStopped true)
  (if (Utilities|Enum|Equal(Enum) (Variables|Effects|GetStopSurfaceType) "SurfaceType8")
    (Networking|CanExecuteCosmeticEvents
      (:True
        (Rendering|SetVisibility (Variables|Projectile_GrenadeLauncher|GetMesh) false false)
        (bind _returnvalue (Effects|SpawnEmitteratLocation (Variables|Effects|GetWaterLandDelayPS) (Variables|Default|GetStopLocation) (Variables|Default|GetStopRotX)))
        (Variables|Default|SetWaterLandDelayRipplePSC _returnvalue)))))

(event Custom|FuseTimerMax
  ;[1 1 1] Bounce
  ;;[1 1 1] DelayExplosion
  (if (not (Delivery|IsBeingKilled))
    (Delivery|DoExplosion)
    (Audio|Components|Audio|FadeOut (Variables|Default|GetSnowball_Loop) 0.3 0.0 "Linear" :!class /Script/Engine.AudioComponent)
    (Fortnite|ShowSoundIndicatorLocal :Location (Transformation|GetActorLocation) :MaxAudibleDistance 5e+03 :IndicatorType "Gunshot" :Instigator self :Tint "(R=0.000000,G=0.000000,B=0.000000,A=0.000000)" :AttenuationMaxRangeMultiplier 1.0 :AttenuationIntensityMultiplier 1.0)))

(event EventOnBounce
  (params (Hit :type Hit%20Result%20Structure%20%28by%20ref%29))
  ;[1 1 1] Bounce
  (Variables|Default|SetOnBounceHitResult Hit)
  (Fortnite|ShowSoundIndicatorLocal :Location (Transformation|GetActorLocation) :MaxAudibleDistance 5e+03 :IndicatorType "Gunshot" :Instigator self :Tint "(R=0.000000,G=0.000000,B=0.000000,A=0.000000)" :AttenuationMaxRangeMultiplier 1.0 :AttenuationIntensityMultiplier 1.0))

(event Custom|DelayBeforeExplode_
  ;[1 1 1] Bounce
  ;;[1 1 1] DelayExplosion
  (Utilities|FlowControl|Delay 0.25)
  (if (not (Delivery|IsBeingKilled))
    (Delivery|DoExplosion)
    (Audio|Components|Audio|FadeOut (Variables|Default|GetSnowball_Loop) 0.3 0.0 "Linear" :!class /Script/Engine.AudioComponent)
    (Fortnite|ShowSoundIndicatorLocal :Location (Transformation|GetActorLocation) :MaxAudibleDistance 5e+03 :IndicatorType "Gunshot" :Instigator self :Tint "(R=0.000000,G=0.000000,B=0.000000,A=0.000000)" :AttenuationMaxRangeMultiplier 1.0 :AttenuationIntensityMultiplier 1.0)))

(event OnEnterWaterEvent(WaterInteractionComponent)
  (params (WaterBody :type Fort%20Water%20Body%20Actor%20Object%20Reference) (WaterInteractionComponent :type Fort%20Water%20Interaction%20Component%20Object%20Reference) (bIsFirstBody :type Boolean))
  ;[1 1 1] Explode on hit
  (Delivery|DoExplosion))

(event Collision|EventHit
  (params (MyComp :type Primitive%20Component%20Object%20Reference) (Other :type Actor%20Object%20Reference) (OtherComp :type Primitive%20Component%20Object%20Reference) (bSelfMoved :type Boolean) (HitLocation :type Vector :default "0, 0, 0") (HitNormal :type Vector :default "0, 0, 0") (NormalImpulse :type Vector :default "0, 0, 0") (Hit :type Hit%20Result%20Structure%20%28by%20ref%29))
  ;[1 1 1] Explode on hit
  (Delivery|DoExplosion))

(event EventOnExploded
  (params (HitActors :type Array%20of%20Actor%20Object%20References) (HitResults :type Array%20of%20Hit%20Result%20Structures))
  (bind _waterinteractioncomponent (Variables|Water|GetWaterInteractionComponent))
  (bind (_waterplanelocation _waterplanenormal _watersurfaceposition _waterdepth _waterbodyidx) (FX|GetLastWaterSurfaceInfo _waterinteractioncomponent :!class /Script/FortniteGame.FortWaterInteractionComponent))
  (bind _returnvalue (Transformation|GetActorLocation))
  ;[1 1 1] WorldObjectSurfaceType, Distinguish between water and everything else
  (Networking|CanExecuteCosmeticEvents
    (:True
      (switch Utilities|FlowControl|Switch|SwitchonEPhysicalSurface (Variables|Effects|GetStopSurfaceType)
        (:SurfaceType_Default
          _)
        (:SurfaceType1
          _)
        (:SurfaceType2
          _)
        (:SurfaceType3
          _)
        (:SurfaceType4
          _)
        (:SurfaceType5
          _)
        (:SurfaceType6
          _)
        (:SurfaceType7
          _)
        (:SurfaceType8
          _)
        (:SurfaceType9
          _)
        (:SurfaceType10
          _)
        (:SurfaceType11
          _)
        (:SurfaceType12
          _)
        (:SurfaceType13
          _)
        (:SurfaceType14
          _)
        (:SurfaceType15
          _)
        (:SurfaceType16
          _)
        (:SurfaceType17
          _)
        (:SurfaceType23
          _)
        (:SurfaceType24
          _))
      (bind _returnvalue_2 (Niagara|SpawnSystematLocation (Variables|Default|GetNS_Explosion) _returnvalue (Variables|Default|GetStopRotZ)))
      (if (Water|IsinWater _waterinteractioncomponent)
        (Utilities|FlowControl|Sequence
          (:then_0
            (Components|Activation|Deactivate (Variables|Default|GetWaterLandDelayRipplePSC))
            (Audio|PlaySoundatLocation (Variables|Sounds|GetGrenade_Explosion_Water_Sound) _returnvalue)
            (Audio|PlaySoundatLocation "/Game/Athena/Sounds/Weapons/Explosions/Expl_Debris_Water_Cue.Expl_Debris_Water_Cue" _returnvalue))
          (:then_1
            (Effects|Components|ParticleSystem|SetBoolParameter _returnvalue_2 "bIsInWater" true)
            (Effects|Components|ParticleSystem|SetFloatParameter _returnvalue_2 "WaterDepth" _waterdepth)
            (Effects|Components|ParticleSystem|SetVectorParameter _returnvalue_2 "WaterPlaneLocation" _waterplanelocation)
            (Effects|Components|ParticleSystem|SetVectorParameter _returnvalue_2 "WaterPlaneNormal" _waterplanenormal)
            (Effects|Components|ParticleSystem|SetFloatParameter _returnvalue_2 "WaterBodyIndex" (Math|Conversions|ToFloat(Integer) _waterbodyidx))))
        (else
          (Audio|PlaySoundatLocation (Variables|Sounds|GetGrenade_Explosion_Sound) _returnvalue)
          (Rendering|SetVisibility (Variables|Projectile_GrenadeLauncher|GetMesh) false false)))
      (GameplayCue|ExecuteGameplayCue%28Burst%29Local self (Variables|Default|GetFeedbackCue) (Ability|GameplayCue|MakeGameplayCueParameters 0.0)) ; Cam shake + controller vibration
      (Audio|Components|Audio|FadeOut (Variables|Default|GetSnowball_Loop) 0.1 0.0 "Linear" :!class /Script/Engine.AudioComponent))))

(fn SpawnBounceFX
  (params (Hit :type Hit%20Result%20Structure%20%28by%20ref%29))
  (bind (_bblockinghit _binitialoverlap _time _distance _location _impactpoint _normal _impactnormal _physmat _hitactor) (Collision|BreakHitResult Hit))
  (bind _returnvalue (Math|Conversions|RotationFromXVector _impactnormal))
  (bind _returnvalue_2 (Collision|GetFortPhysicalSurface Hit))
  (switch Utilities|FlowControl|Switch|SwitchonEPhysicalSurface _returnvalue_2
    (:SurfaceType_Default
      _)
    (:SurfaceType1
      _)
    (:SurfaceType2
      _)
    (:SurfaceType3
      _)
    (:SurfaceType4
      _)
    (:SurfaceType5
      _)
    (:SurfaceType6
      _)
    (:SurfaceType7
      _)
    (:SurfaceType8
      (Effects|SpawnEmitteratLocation (Variables|Effects|GetBounce_WaterSplash_PS) _impactpoint _returnvalue)
      (Audio|PlaySoundatLocation "/Game/Sounds/Fort_GamePlay_Sounds/Explosives/Snowball_Projectile_Bounce_Cue.Snowball_Projectile_Bounce_Cue" _impactpoint)
      (return _hitactor _returnvalue_2 _returnvalue))
    (:SurfaceType9
      _)
    (:SurfaceType10
      _)
    (:SurfaceType11
      _)
    (:SurfaceType12
      _)
    (:SurfaceType13
      _)
    (:SurfaceType14
      _)
    (:SurfaceType15
      _)
    (:SurfaceType16
      _)
    (:SurfaceType17
      _))
  (Effects|SpawnEmitteratLocation (Variables|Effects|GetBounce_Generic_PS) _impactpoint _returnvalue)
  (Audio|PlaySoundatLocation "/Game/Sounds/Fort_GamePlay_Sounds/Explosives/Snowball_Projectile_Bounce_Cue.Snowball_Projectile_Bounce_Cue" _impactpoint))

(fn ConstructionScript
  (Effects|Components|ParticleSystem|SetTemplate (Variables|Default|GetRibbonTrail_PSC) (Variables|Effects|GetRibbonTrail_PS))
  (Effects|Components|ParticleSystem|SetTemplate (Variables|Default|GetWaterSplash_PSC) (Variables|Effects|GetBounce_WaterSplash_PS))
  (Effects|Components|ParticleSystem|SetTemplate (Variables|Default|GetBounce_Generic_PSC) (Variables|Effects|GetBounce_Generic_PS))
  (Effects|Components|ParticleSystem|SetTemplate (Variables|Default|GetExplosion_PSC) (Variables|Effects|GetExplosion_PS))
  (Effects|Components|ParticleSystem|SetTemplate (Variables|Default|GetExplosion_Water_PSC) 0)
  (AddComponent|NODE_AddParticleSystemComponent-1))
